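# Control Rig for the collar part of the female Future Orange Graffiti body, rotating collar bones procedurally as the head turns relative to the neck.
# Each step reads head rotation in neck_01 space, remaps its Euler Z and adds it as local Y rotation on a collar MSTR bone.
# Collars a, c and d are driven at differing amounts; b's zero output range leaves it unrotated, and e has no step.
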ControlRigBlueprint /BRCosmetics/Characters/Player/Female/Medium/Bodies/F_MED_Graffiti_Future_Orange/Meshes/Parts/Future_Orange_Collar_CtrlRig
graphs [
 RigVMFunctionLibrary {links [], nodes []}
 RigVMModel {nodes [
   RigUnit_BeginExecution = &RigUnit_BeginExecution()
   n1 = &RigVMFunction_MathQuaternionToEuler(RotationOrder=ZYX, Value=&RigUnit_GetRelativeBoneTransform(Bone=head, Space=neck_01).Transform.Rotation)
   /Script/ControlRig.RigUnit_ModifyBoneTransforms(
    BoneToModify=((Bone="dyn_collar_a_MSTR",Transform=(Rotation=(X=0,Y=0,Z=0,W=1),Translation=(X=0,Y=0,Z=0),Scale3D=(X=1,Y=1,Z=1))))
    Mode=AdditiveLocal
    Weight=1
    WeightMaximum=1
    BoneToModify.0.Transform.Rotation=&RigVMFunction_MathQuaternionFromEuler(
     Euler=(X=0,Y=0,Z=0)
     Euler.Y=&RigVMFunction_MathFloatRemap(SourceMaximum=-50, SourceMinimum=10, TargetMaximum=10, TargetMinimum=0, Value=@n1.Result.Z).Result
    ).Result
    ExecutePin=@RigUnit_BeginExecution.ExecutePin
   ) {is=struct RigVMUnitNode}
   n2 = &RigVMFunction_MathQuaternionToEuler(RotationOrder=ZYX, Value=&RigUnit_GetRelativeBoneTransform(Bone=head, Space=neck_01).Transform.Rotation)
   > &RigUnit_ModifyBoneTransforms(
    BoneToModify=((Bone="dyn_collar_b_MSTR",Transform=(Rotation=(X=0,Y=0,Z=0,W=1),Translation=(X=0,Y=0,Z=0),Scale3D=(X=1,Y=1,Z=1))))
    Mode=AdditiveLocal
    Weight=1
    WeightMaximum=1
    BoneToModify.0.Transform.Rotation=&RigVMFunction_MathQuaternionFromEuler(
     Euler=(X=0,Y=0,Z=0)
     Euler.Y=&RigVMFunction_MathFloatRemap(SourceMaximum=-50, SourceMinimum=10, TargetMaximum=0, TargetMinimum=0, Value=@n2.Result.Z).Result
    ).Result
   )
   n3 = &RigVMFunction_MathQuaternionToEuler(RotationOrder=ZYX, Value=&RigUnit_GetRelativeBoneTransform(Bone=head, Space=neck_01).Transform.Rotation)
   > &RigUnit_ModifyBoneTransforms(
    BoneToModify=((Bone="dyn_collar_c_MSTR",Transform=(Rotation=(X=0,Y=0,Z=0,W=1),Translation=(X=0,Y=0,Z=0),Scale3D=(X=1,Y=1,Z=1))))
    Mode=AdditiveLocal
    Weight=1
    WeightMaximum=1
    BoneToModify.0.Transform.Rotation=&RigVMFunction_MathQuaternionFromEuler(
     Euler=(X=0,Y=0,Z=0)
     Euler.Y=&RigVMFunction_MathFloatRemap(SourceMaximum=-50, SourceMinimum=10, TargetMaximum=-100, TargetMinimum=0, Value=@n3.Result.Z).Result
    ).Result
   )
   n4 = &RigVMFunction_MathQuaternionToEuler(RotationOrder=ZYX, Value=&RigUnit_GetRelativeBoneTransform(Bone=head, Space=neck_01).Transform.Rotation)
   > &RigUnit_ModifyBoneTransforms(
    BoneToModify=((Bone="dyn_collar_d_MSTR",Transform=(Rotation=(X=0,Y=0,Z=0,W=1),Translation=(X=0,Y=0,Z=0),Scale3D=(X=1,Y=1,Z=1))))
    Mode=AdditiveLocal
    Weight=1
    WeightMaximum=1
    BoneToModify.0.Transform.Rotation=&RigVMFunction_MathQuaternionFromEuler(
     Euler=(X=0,Y=0,Z=0)
     Euler.Y=&RigVMFunction_MathFloatRemap(SourceMaximum=-50, SourceMinimum=10, TargetMaximum=-100, TargetMinimum=0, Value=@n4.Result.Z).Result
    ).Result
   )
  ]}
]
hierarchy [
 root {loc=<0 0 0>, rot=<0 0 0 1>}
 attach {parent=root, loc=<0 0 0>, rot=<0 0 0 1>}
 pelvis {parent=root, loc=<0 -0.5301 98.0571>, rot=<0.0127 0.707 -0.0127 -0.707>}
 spine_01 {parent=pelvis, loc=<6.1208 -0.9328 0>, rot=<0 0 0.034 -0.9994>}
 spine_02 {parent=spine_01, loc=<6.5996 -0.2095 0>, rot=<0 0 0.0238 -0.9997>}
 spine_03 {parent=spine_02, loc=<6.4914 -0.1237 0>, rot=<0 0 0.0149 -0.9999>}
 spine_04 {parent=spine_03, loc=<6.9598 0.0794 0>, rot=<0 0 -0.0132 -0.9999>}
 spine_05 {parent=spine_04, loc=<6.9987 0 0>, rot=<0 0 -0.0015 -1>}
 clavicle_l {parent=spine_05, loc=<5.4831 -0.1053 -3.2946>, rot=<0.0465 -0.7289 -0.009 -0.6829>}
 clavicle_r {parent=spine_05, loc=<5.4831 -0.1053 3.2946>, rot=<0.7289 0.0465 -0.6829 0.009>}
 neck_01 {parent=spine_05, loc=<11.3138 0.1529 0>, rot=<0 0 -0.2003 -0.9797>}
 neck_02 {parent=neck_01, loc=<4.5627 0 0>, rot=<0 0 0.0504 -0.9987>}
 head {parent=neck_02, loc=<5.3914 0 0>, rot=<0 0 0.115 -0.9934>}
 dyn_SideBang_MSTR {parent=head, loc=<4.956 4.2416 -6.0941>, rot=<0.0737 -0.9909 -0.0595 -0.0952>}
 dyn_SideBang_01 {parent=dyn_SideBang_MSTR, loc=<0 0 0>, rot=<0 0 0 -1>}
 dyn_SideBang_02 {parent=dyn_SideBang_01, loc=<4.0129 0 0>, rot=<-0.6792 -0.1106 0.2034 -0.6964>}
 dyn_SideBang_03 {parent=dyn_SideBang_02, loc=<3.1627 0 0>, rot=<0.6921 0.0646 0.083 -0.7141>}
 dyn_BackLeftHair_MSTR {parent=head, loc=<3.3704 -3.4847 -5.172>, rot=<-0.0001 -0.963 0.2695 -0.0093>}
 dyn_BackLeftHair_01 {parent=dyn_BackLeftHair_MSTR, loc=<0 0 0>, rot=<0 0 0 -1>}
 dyn_BackLeftHair_02 {parent=dyn_BackLeftHair_01, loc=<3.1282 0 0>, rot=<0 -0.0092 0 -1>}
 dyn_BackRightHair_MSTR {parent=head, loc=<3.1514 -3.4865 4.8521>, rot=<-0.0083 -0.2166 0.9762 -0.0043>}
 dyn_BackRightHair_01 {parent=dyn_BackRightHair_MSTR, loc=<0 0 0>, rot=<0 0 0 1>}
 dyn_BackRightHair_02 {parent=dyn_BackRightHair_01, loc=<3.1282 0 0>, rot=<0 -0.0092 0 -1>}
 dyn_BackHair_MSTR {parent=head, loc=<2.5487 -5.9429 0>, rot=<-0.0047 -0.7054 0.7087 -0.0081>}
 dyn_BackHair_01 {parent=dyn_BackHair_MSTR, loc=<0 0 0>, rot=<0 0 0 -1>}
 dyn_BackHair_02 {parent=dyn_BackHair_01, loc=<3.1282 0 0>, rot=<0 -0.0092 0 -1>}
 dyn_FrontBang_MSTR {parent=head, loc=<5.2075 8.2738 -4.3004>, rot=<0.1689 -0.9201 -0.2545 -0.2451>}
 dyn_FrontBang_01 {parent=dyn_FrontBang_MSTR, loc=<0 0 0>, rot=<0 0 0 1>}
 dyn_FrontBang_02 {parent=dyn_FrontBang_01, loc=<3.1339 0 0>, rot=<-0.0314 -0.1681 0.1122 -0.9789>}
 dyn_FrontBang_03 {parent=dyn_FrontBang_02, loc=<2.7804 0 0>, rot=<-0.0488 -0.0469 0.0245 -0.9974>}
 dyn_collar_e_MSTR {parent=spine_05, loc=<7.6348 4.6977 5.5517>, rot=<0.0494 0.2773 -0.2092 -0.9364>}
 dyn_collar_e {parent=dyn_collar_e_MSTR, loc=<0 0 0>, rot=<0 0 0 1>}
 dyn_collar_d_MSTR {parent=spine_05, loc=<11.4769 -2.2246 5.6447>, rot=<-0.3787 0.0568 -0.1577 -0.9102>}
 dyn_collar_d {parent=dyn_collar_d_MSTR, loc=<0 0 0>, rot=<0 0 0 1>}
 dyn_collar_c_MSTR {parent=spine_05, loc=<12.3318 -3.833 0>, rot=<-0.7065 -0.0284 -0.0284 -0.7065>}
 dyn_collar_c {parent=dyn_collar_c_MSTR, loc=<0 0 0>, rot=<0 0 0 -1>}
 dyn_collar_b_MSTR {parent=spine_05, loc=<11.4769 -2.2246 -5.6447>, rot=<0.918 0.1565 -0.0601 0.3594>}
 dyn_collar_b {parent=dyn_collar_b_MSTR, loc=<0 0 0>, rot=<0 0 0 -1>}
 dyn_collar_a_MSTR {parent=spine_05, loc=<7.6348 4.6977 -5.5517>, rot=<0.9363 0.2083 -0.278 -0.0526>}
 dyn_collar_a {parent=dyn_collar_a_MSTR, loc=<0 0 0>, rot=<0 0 0 1>}
]
