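# Physics asset for the hair of the Future Orange graffiti female medium player body, covering head, neck and dynamic hair bones.
# Five chains (back, back-left, back-right, front bang, side bang) use box bodies, each rooted at a tiny capsule master body constrained to the head.
# Head, neck and master-root joints swing and twist freely, while joints along the chains are angle-limited, tighter in swing for back hair than for bangs.
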PhysicsAsset /BRCosmetics/Characters/Player/Female/Medium/Bodies/F_MED_Graffiti_Future_Orange/Meshes/Parts/Future_Orange_Physics
bodies {
 dyn_BackHair_01 [{shapeType=Box, extentX=3.1345, extentY=7.0745, extentZ=3.2177, center=<1.6017 0 0>, rotation=<rot 90 -13.2613 -13.357>}]
 dyn_BackHair_02 [{shapeType=Box, extentX=4.5869, extentY=7.0745, extentZ=3.2177, center=<1.5479 0 0>, rotation=<rot 90 13.2613 13.1657>}]
 dyn_BackHair_MSTR [{shapeType=Capsule, length=0.505, radius=0.505, rotation=<rot 0 -90 0>}]
 dyn_BackLeftHair_01 [{shapeType=Box, extentX=2.0532, extentY=7.0745, extentZ=3.2177, center=<1.5183 0 0>, rotation=<rot 90 13.2613 13.1656>}]
 dyn_BackLeftHair_02 [{shapeType=Box, extentX=3.9216, extentY=7.0745, extentZ=3.2177, center=<1.5025 0 0>, rotation=<rot 90 13.2613 13.1656>}]
 dyn_BackLeftHair_MSTR [{shapeType=Capsule, length=0.505, radius=0.505, rotation=<rot 0 -90 0>}]
 dyn_BackRightHair_01 [{shapeType=Box, extentX=1.5786, extentY=7.0745, extentZ=3.2177, center=<1.7145 0 0>, rotation=<rot 90 13.2613 13.1657>}]
 dyn_BackRightHair_02 [{shapeType=Box, extentX=3.2078, extentY=5.3526, extentZ=3.2177, center=<1.7145 0 0>, rotation=<rot 90 13.2613 13.1657>}]
 dyn_BackRightHair_MSTR [{shapeType=Capsule, length=0.505, radius=0.505, rotation=<rot 0 -90 0>}]
 dyn_FrontBang_01 [{shapeType=Box, extentX=3.2245, extentY=3.6566, extentZ=2.2723, center=<1.625 0 0.0001>}]
 dyn_FrontBang_02 [{shapeType=Box, extentX=2.5613, extentY=2.9775, extentZ=0.7989, center=<1.3521 0 0.0001>}]
 dyn_FrontBang_03 [{shapeType=Box, extentX=2.5613, extentY=2.9775, extentZ=0.7989, center=<1.3521 0 0.0001>}]
 dyn_FrontBang_MSTR [{shapeType=Capsule, length=0.505, radius=0.505, rotation=<rot 0 -90 0>}]
 dyn_SideBang_01 [{shapeType=Box, extentX=1.5598, extentY=2.8217, extentZ=3.8849, center=<1.987 0 0>, rotation=<rot 83.1252 89.9964 89.9007>}]
 dyn_SideBang_02 [{shapeType=Box, extentX=3.2365, extentY=3.5727, extentZ=0.9826, center=<1.6325 0 0>}]
 dyn_SideBang_03 [{shapeType=Box, extentX=0.8406, extentY=2.8217, extentZ=3.8849, center=<1.9182 -0.0218 0.0029>, rotation=<rot 83.1252 89.9964 89.9007>}]
 dyn_SideBang_MSTR [{shapeType=Capsule, length=0.505, radius=0.505, rotation=<rot 0 -90 0>}]
 head [{shapeType=Capsule, length=4.6379, radius=5.8818, center=<-0.222 2.2796 0.0987>, rotation=<rot 1.8525 -90.1791 -78.2549>}]
 neck_01 [{shapeType=Capsule, length=6.4114, radius=2.9941}]
 neck_02 [{shapeType=Capsule, length=3.5134, radius=2.9941}]
}
constraints [
 {
  bone1Name=dyn_SideBang_MSTR
  bone2Name=head
  swing1LimitDegrees=45
  swing1Motion=Free
  swing2LimitDegrees=45
  swing2Motion=Free
  twistLimitDegrees=45
  twistMotion=Free
 }
 {
  bone1Name=dyn_BackLeftHair_MSTR
  bone2Name=head
  swing1LimitDegrees=45
  swing1Motion=Free
  swing2LimitDegrees=45
  swing2Motion=Free
  twistLimitDegrees=45
  twistMotion=Free
 }
 {
  bone1Name=dyn_BackRightHair_MSTR
  bone2Name=head
  swing1LimitDegrees=45
  swing1Motion=Free
  swing2LimitDegrees=45
  swing2Motion=Free
  twistLimitDegrees=45
  twistMotion=Free
 }
 {
  bone1Name=dyn_BackHair_MSTR
  bone2Name=head
  swing1LimitDegrees=45
  swing1Motion=Free
  swing2LimitDegrees=45
  swing2Motion=Free
  twistLimitDegrees=45
  twistMotion=Free
 }
 {
  bone1Name=dyn_FrontBang_MSTR
  bone2Name=head
  swing1LimitDegrees=45
  swing1Motion=Free
  swing2LimitDegrees=45
  swing2Motion=Free
  twistLimitDegrees=45
  twistMotion=Free
 }
 {bone1Name=dyn_SideBang_01, bone2Name=dyn_SideBang_MSTR, swing1LimitDegrees=45, swing2LimitDegrees=45, twistLimitDegrees=15}
 {bone1Name=dyn_SideBang_03, bone2Name=dyn_SideBang_02, swing1LimitDegrees=45, swing2LimitDegrees=45, twistLimitDegrees=15}
 {bone1Name=dyn_BackLeftHair_01, bone2Name=dyn_BackLeftHair_MSTR, swing1LimitDegrees=10, swing2LimitDegrees=10, twistLimitDegrees=10}
 {bone1Name=dyn_BackLeftHair_02, bone2Name=dyn_BackLeftHair_01, swing1LimitDegrees=10, swing2LimitDegrees=10, twistLimitDegrees=10}
 {bone1Name=dyn_BackRightHair_01, bone2Name=dyn_BackRightHair_MSTR, swing1LimitDegrees=10, swing2LimitDegrees=10, twistLimitDegrees=10}
 {bone1Name=dyn_BackRightHair_02, bone2Name=dyn_BackRightHair_01, swing1LimitDegrees=10, swing2LimitDegrees=10, twistLimitDegrees=10}
 {bone1Name=dyn_BackHair_01, bone2Name=dyn_BackHair_MSTR, swing1LimitDegrees=10, swing2LimitDegrees=10, twistLimitDegrees=10}
 {bone1Name=dyn_BackHair_02, bone2Name=dyn_BackHair_01, swing1LimitDegrees=10, swing2LimitDegrees=10, twistLimitDegrees=10}
 {
  bone1Name=head
  bone2Name=neck_02
  swing1LimitDegrees=45
  swing1Motion=Free
  swing2LimitDegrees=45
  swing2Motion=Free
  twistLimitDegrees=45
  twistMotion=Free
 }
 {
  bone1Name=neck_02
  bone2Name=neck_01
  swing1LimitDegrees=45
  swing1Motion=Free
  swing2LimitDegrees=45
  swing2Motion=Free
  twistLimitDegrees=45
  twistMotion=Free
 }
 {bone1Name=dyn_SideBang_02, bone2Name=dyn_SideBang_01, swing1LimitDegrees=45, swing2LimitDegrees=45, twistLimitDegrees=15}
 {bone1Name=dyn_FrontBang_01, bone2Name=dyn_FrontBang_MSTR, swing1LimitDegrees=45, swing2LimitDegrees=45, twistLimitDegrees=5}
 {bone1Name=dyn_FrontBang_02, bone2Name=dyn_FrontBang_01, swing1LimitDegrees=45, swing2LimitDegrees=45, twistLimitDegrees=5}
 {bone1Name=dyn_FrontBang_03, bone2Name=dyn_FrontBang_02, swing1LimitDegrees=45, swing2LimitDegrees=45, twistLimitDegrees=5}
]
