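# Post-process-style Anim Blueprint for the Future Orange female graffiti body part, copying the attached parent's pose and adding collar and hair secondary motion.
# While crouch-moving, it blends in additive back-collar offsets scaled by head yaw relative to the neck.
# Cross-linked rotation constraints couple front and side bangs ahead of RigidBody simulation.
# A right-arm-driven collar/hood correction and collar Control Rig finish the pose, using alphas computed from ref-pose bone deltas.
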AnimBlueprint /BRCosmetics/Characters/Player/Female/Medium/Bodies/F_MED_Graffiti_Future_Orange/Meshes/Parts/F_MED_Graffiti_Future_Orange_AnimBP
target_skeleton=&/F_MED_Graffiti_Future_Orange_Skeleton
graphs [
 AnimGraph {nodes [
   n1 = Get BackCollar_Fix() -> (BackCollar_Fix:double)
   n2 = "Transform (Modify) Bone - Bone: dyn_collar_b_MSTR"(
    ComponentPose=Local To Component(LocalPose=Copy Pose From Mesh() {bCopyCurves=true, bUseAttachedParent=true})
    Alpha=@n1
    Translation="0,-7,-2"
    Rotation="-50,0.000,0.000"
    Scale="1,1,1"
   ) {
    rotationMode=BMM_Additive
    scaleMode=BMM_Additive
    translationMode=BMM_Additive
    alphaBoolBlend {blendOption=Linear}
    boneToModify {boneName=dyn_collar_b_MSTR}
    rotation {pitch=-50}
    translation {y=-7, z=-2}
   }
   n3 = "Transform (Modify) Bone - Bone: dyn_collar_c_MSTR"(ComponentPose=@n2, Alpha=@n1, Translation="0,-7,-2", Rotation="-20,0,0", Scale="1,1,1") {
    rotationMode=BMM_Additive
    scaleMode=BMM_Additive
    translationMode=BMM_Additive
    alphaBoolBlend {blendOption=Linear}
    boneToModify {boneName=dyn_collar_c_MSTR}
    rotation {pitch=-20}
    translation {y=-7, z=-2}
   }
   n4 = "Transform (Modify) Bone - Bone: dyn_collar_d_MSTR"(ComponentPose=@n3, Alpha=@n1, Translation="0,-7,-2", Rotation="0,0,0", Scale="1,1,1") {
    rotationMode=BMM_Additive
    scaleMode=BMM_Additive
    translationMode=BMM_Additive
    alphaBoolBlend {blendOption=Linear}
    boneToModify {boneName=dyn_collar_d_MSTR}
    translation {y=-7, z=-2}
   }
   n5 = Component To Local(ComponentPose=@n4)
   n6 = Copy Pose From Mesh() {bCopyCurves=true, bUseAttachedParent=true}
   n7 = AND Boolean(A=Get bIsCrouching(), B=Get bIsMoving2D())
   n8 = "Constraint - dyn_FrontBang_01 (R)"(
    ComponentPose=Local To Component(LocalPose=Blend Poses by bool(BlendPose_0=@n5, BlendPose_1=@n6, BlendTime_0=0.2, BlendTime_1=0.2, bActiveValue=@n7) {blendType=Linear})
    Alpha=0.4
    ConstraintWeights_0=1
   ) {
    alphaBoolBlend {blendOption=Linear}
    boneToModify {boneName=dyn_FrontBang_01}
    constraintSetup [{transformType=Rotation, targetBone {boneName=dyn_SideBang_01}}]
    constraintWeights [1]
   }
   n9 = "Constraint - dyn_FrontBang_02 (R)"(ComponentPose=@n8, Alpha=0.3, ConstraintWeights_0=1) {
    alphaBoolBlend {blendOption=Linear}
    boneToModify {boneName=dyn_FrontBang_02}
    constraintSetup [{transformType=Rotation, targetBone {boneName=dyn_SideBang_02}}]
    constraintWeights [1]
   }
   n10 = "Constraint - dyn_FrontBang_03 (R)"(ComponentPose=@n9, Alpha=0.2, ConstraintWeights_0=1) {
    alphaBoolBlend {blendOption=Linear}
    boneToModify {boneName=dyn_FrontBang_03}
    constraintSetup [{transformType=Rotation, targetBone {boneName=dyn_SideBang_03}}]
    constraintWeights [1]
   }
   n11 = "Constraint - dyn_SideBang_01 (R)"(ComponentPose=@n10, Alpha=0.5, ConstraintWeights_0=1) {
    alphaBoolBlend {blendOption=Linear}
    boneToModify {boneName=dyn_SideBang_01}
    constraintSetup [{transformType=Rotation, targetBone {boneName=dyn_FrontBang_01}}]
    constraintWeights [1]
   }
   n12 = "Constraint - dyn_SideBang_02 (R)"(ComponentPose=@n11, Alpha=0.4, ConstraintWeights_0=1) {
    alphaBoolBlend {blendOption=Linear}
    boneToModify {boneName=dyn_SideBang_02}
    constraintSetup [{transformType=Rotation, targetBone {boneName=dyn_FrontBang_02}}]
    constraintWeights [1]
   }
   n13 = "Constraint - dyn_SideBang_03 (R)"(ComponentPose=@n12, Alpha=0.3, ConstraintWeights_0=1) {
    alphaBoolBlend {blendOption=Linear}
    boneToModify {boneName=dyn_SideBang_03}
    constraintSetup [{transformType=Rotation, targetBone {boneName=dyn_FrontBang_03}}]
    constraintWeights [1]
   }
   n14 = RigidBody(ComponentPose=@n13) {
    bClampLinearTranslationLimitToRefPose=true
    bOverrideWorldGravity=true
    overridePhysicsAsset=&/Future_Orange_Physics
    simulationSpace=BaseBoneSpace
    alphaBoolBlend {blendOption=Linear}
    baseBoneRef {boneName=root}
    componentAppliedLinearAccClamp=<5000 5000 5000>
    componentLinearAccScale=<0.4 0.4 0.5>
    overrideWorldGravity {z=-980}
   }
   n15 = "Transform (Modify) Bone - Bone: dyn_collar_e_MSTR"(ComponentPose=@n14, Alpha=Get hoodFix_alpha_r() -> (hoodFix_alpha_r:double), Translation="0,0,0", Rotation="-46,0,0", Scale="1,1,1") {
    rotationMode=BMM_Additive
    rotationSpace=BCS_BoneSpace
    scaleMode=BMM_Additive
    scaleSpace=BCS_BoneSpace
    translationMode=BMM_Additive
    translationSpace=BCS_BoneSpace
    alphaBoolBlend {blendOption=Linear}
    boneToModify {boneName=dyn_collar_e_MSTR}
    rotation {pitch=-46}
   }
   n16 = Control Rig(Source=Component To Local(ComponentPose=@n15)) {
    alphaBoolBlend {blendOption=Linear}
    controlRigAssetReference {blueprintRigClass=&/Future_Orange_Collar_CtrlRig.Future_Orange_Collar_CtrlRig_C}
   }
   Output Pose(Result=@n16) {layerGroup=None}
  ]}
 EventGraph {nodes [
   Event BlueprintUpdateAnimation()
   > n1 = Sequence()
   n2 = GetOwnerSkeletalMeshComponent()
   n3 = BreakRotator(InRot=BreakTransform(InTransform=GetDeltaTransformFromRefPose(self=@n2, BoneName=upperarm_r, BaseName=clavicle_r)).Rotation)
   > Set armForward_R(armForward_R:double=@n3.Yaw) -> (Output_Get:double)
   > Set armUp_R(armUp_R:double=@n3.Pitch) -> (Output_Get:double)
   n4 = BreakTransform(InTransform=GetDeltaTransformFromRefPose(self=GetOwnerSkeletalMeshComponent(), BoneName=head, BaseName=neck_02))
   > Set BackCollar_Fix(BackCollar_Fix:double=MapRangeClamped(Value=BreakRotator(InRot=@n4.Rotation).Yaw, InRangeA=0, InRangeB=-40)) -> (Output_Get:double)
   n5 = BreakTransform(InTransform=GetDeltaTransformFromRefPose(self=GetOwnerSkeletalMeshComponent(), BoneName=neck_02, BaseName=spine_05))
   > Set BendOver(BendOver:double=MapRangeClamped(Value=BreakRotator(InRot=@n5.Rotation).Yaw, InRangeA=-10, InRangeB=-20)) -> (Output_Get:double)
   n6 = Get armUp_R() -> (armUp_R:double)
   n7 = Get armForward_R() -> (armForward_R:double)
   Set hoodFix_alpha_r(
    execute=@n1.then_1
    hoodFix_alpha_r:double=float * float(A=MapRangeClamped(Value=@n6, InRangeA=-9, InRangeB=20), B=MapRangeClamped(Value=@n7, InRangeA=70, InRangeB=90))
   ) -> (Output_Get:double)
  ]}
]
variables [
 armForward_R {type=number}
 armUp_R {type=number}
 BackCollar_Fix {type=number}
 BendOver {type=number}
 hoodFix_alpha_r {type=number}
 LookUp {type=number}
]
